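# Material function for Princess Castle props that adds a slow, strobe-like flicker to incoming Material Attributes by multiplying emissive color, behind an off-by-default UseEmissiveFlicker switch.
# By default, two summed sines of different periods drive the flicker, with speed varied per actor by hashing world position (X twice plus Z, ignoring Y).
# Switching off UseFlickerSpeedVariation gives a single sine pulse at max speed above a minimum brightness.
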MaterialFunction /Game/Environments/Sets/PrincessCastle/Props/Materials/MF_GlowFlicker
graph {
 properties {description=This function is gives assets a slow strobe-like flicker on top of it's material.}
 nodes [
  n1 = FunctionInput() {inputName=MA, inputType=FunctionInput_MaterialAttributes}
  n2 = GetMaterialAttributes(None=@n1) {attributeGetTypes [#1]}
  n3 = Time()
  n4 = Multiply(A=@n3, B=ScalarParameter() {defaultValue=1.5, group=Emissive, parameterName=FlickerMaxSpeed}) {constB=1.5}
  n5 = MaterialFunctionCall(Float3=ActorPositionWS()) {materialFunction=&BreakOutFloat3Components}
  n6 = Add(A=Add(A=@n5.R, B=@n5.R), B=@n5.B)
  n7 = LinearInterpolate(
   A=Multiply(A=@n3, B=ScalarParameter() {defaultValue=0.7, group=Emissive, parameterName=FlickerMinSpeed}) {constB=0.7}
   B=@n4
   Alpha=Frac(None=Divide(A=@n6) {constB=50})
  )
  n8 = Multiply(
   A=Add(A=Sine(None=@n7) {period=0.5651}, B=Sine(None=@n7) {period=0.442})
   B=ScalarParameter() {defaultValue=0.15, group=Emissive, parameterName=FlickerScale}
  )
  n9 = ScalarParameter() {defaultValue=0.25, group=Emissive, parameterName=FlickerMinBrightnessBias}
  n10 = ConstantBiasScale(None=Sine(None=@n4))
  n11 = Multiply(
   A=@n2.EmissiveColor
   B=StaticSwitchParameter(True=Add(A=@n8), False=LinearInterpolate(A=@n9, Alpha=@n10)) {defaultValue=true, group=Emissive, parameterName=UseFlickerSpeedVariation}
  )
  FunctionOutput(None=StaticSwitchParameter(
    True=SetMaterialAttributes!(MaterialAttributes=@n2.MaterialAttributes, EmissiveColor=@n11) {attributeSetTypes [#1]}
    False=@n1
   ) {group=Emissive, parameterName=UseEmissiveFlicker})
 ]
}
